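# Physics asset for the ScorpionZero backpack cosmetic mesh, set up for dangling left and right straps.
# Each strap is a three-capsule chain constrained from a strap master bone, with limited swing and almost no twist.
# Unconstrained capsules on the pack, spine and left upper arm give body volumes, likely for the straps to collide against.
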PhysicsAsset /BRCosmetics/Accessories/FORT_Backpacks/Backpack_M_MED_ScorpionZero/Meshes/M_MED_ScorpionZero_Pack_PhysicsAsset
bodies {
 dyn_strap_01_L [{shapeType=Capsule, length=4.1, radius=0.8, shapeName=dyn_strap_01_L_capsule, center=<2.54 0 0>, rotation=<rot 90 -0.0293 -0.0297>}]
 dyn_strap_01_R [{shapeType=Capsule, length=4.1, radius=0.95, shapeName=dyn_strap_01_R_capsule, center=<2.891 0 0>, rotation=<rot 90 -6.7215 -6.7219>}]
 dyn_strap_02_L [{shapeType=Capsule, length=9, radius=1.5, shapeName=dyn_strap_02_L_capsule, center=<0.7336 0 0>, rotation=<rot 90 -0.1581 -0.1583>}]
 dyn_strap_02_R [{shapeType=Capsule, length=7, radius=1.5, shapeName=dyn_strap_02_R_capsule, center=<0.8558 0 0>, rotation=<rot 90 -90 -90.0002>}]
 dyn_strap_03_L [{shapeType=Capsule, length=6, radius=1.5, shapeName=dyn_strap_03_L_capsule, center=<3.2918 0 0>, rotation=<rot 90 -0.1581 -0.1583>}]
 dyn_strap_03_R [{shapeType=Capsule, length=6, radius=1.5, shapeName=dyn_strap_03_R_capsule, center=<1.8907 0 0>, rotation=<rot 90 -160.5288 -160.529>}]
 pack [{shapeType=Capsule, length=55.9015, radius=2.5, shapeName=pack_capsule, center=<-0.7032 0.5852 -6.7312>, rotation=<rot 3.8325 -90 35>}]
 spine_01 [{shapeType=Capsule, length=14, radius=11.5, shapeName=spine_01_capsule, center=<1.4991 0 0>}]
 spine_04 [{shapeType=Capsule, length=17, radius=10, shapeName=spine_04_capsule, center=<-0.0671 0 0>}]
 strap_MSTR_L [{shapeType=Capsule, length=0, radius=0.5, shapeName=strap_MSTR_L_capsule, rotation=<rot 90 -90.0001 -90.0005>}]
 strap_MSTR_R [{shapeType=Capsule, length=0, radius=0.5, shapeName=strap_MSTR_R_capsule, rotation=<rot 90 -90 -90.0004>}]
 upperarm_l [{shapeType=Capsule, length=17, radius=9.5, shapeName=upperarm_l_capsule, center=<7.3398 0.0001 0>, rotation=<rot 90 180 179.9995>}]
}
constraints [
 {bone1Name=dyn_strap_01_L, bone2Name=strap_MSTR_L, swing1LimitDegrees=55, swing2LimitDegrees=20, twistLimitDegrees=1}
 {bone1Name=dyn_strap_01_R, bone2Name=strap_MSTR_R, swing1LimitDegrees=55, swing2LimitDegrees=20, twistLimitDegrees=0}
 {bone1Name=dyn_strap_02_L, bone2Name=dyn_strap_01_L, swing1LimitDegrees=45, swing2LimitDegrees=45, twistLimitDegrees=2}
 {bone1Name=dyn_strap_03_L, bone2Name=dyn_strap_02_L, swing1LimitDegrees=45, swing2LimitDegrees=45, twistLimitDegrees=2}
 {bone1Name=dyn_strap_02_R, bone2Name=dyn_strap_01_R, swing1LimitDegrees=45, swing2LimitDegrees=45, twistLimitDegrees=2}
 {bone1Name=dyn_strap_03_R, bone2Name=dyn_strap_02_R, swing1LimitDegrees=45, swing2LimitDegrees=45, twistLimitDegrees=2}
]
