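# Animation Blueprint for the ScorpionZero backpack cosmetic, simulating dangling straps and toggling a material-hidden element.
# Each update copies the body mesh's spine and left upper-arm transforms onto the pack skeleton, feeding strap collision limits, and refreshes gravity override.
# The cached pose feeds either RigidBody or a six-node AnimDynamics chain, selected by bRigidBodyEnabled.
# Customization, weapon and skydiving events set "Hide Element 01" while a ScorpionZero-tagged pickaxe is equipped outside skydiving.
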AnimBlueprint /BRCosmetics/Accessories/FORT_Backpacks/Backpack_M_MED_ScorpionZero/Meshes/M_MED_ScorpionZero_Pack_AnimBP
target_skeleton=&/M_MED_ScorpionZero_Pack_Skeleton
graphs [
 AnimGraph {nodes [
   n1 = Get Gravity Override()
   n2 = Anim Dynamics - Bone: dyn_strap_01_L!(
    ComponentPose=Local To Component(LocalPose=Use cached pose 'before simulation'())
    Alpha=1
    GravityOverride:Vector=@n1
    ExternalForce="0,0,0"
   ) -> (Pose) {
    angularBiasOverride=1
    angularDampingOverride=1
    angularSpringConstant=1000
    bAngularSpring=true
    bOverrideAngularBias=true
    bOverrideAngularDamping=true
    bUseGravityOverride=true
    simulationSpace=RootRelative
    alphaBoolBlend {blendOption=Linear}
    boundBone {boneName=dyn_strap_01_L}
    componentAppliedLinearAccClamp=<3000 3000 3000>
    componentLinearAccScale=<0.4 0.5 0.5>
    componentLinearVelScale=<0.5 0.5 0.5>
    physicsBodyDefinitions [{
      collisionType=CustomSphere
      sphereCollisionRadius=2.5
      boundBone {boneName=dyn_strap_01_L}
      boxExtents=<16 12 12>
      constraintSetup {angularLimitsMax=<0 45 10>, angularLimitsMin=<0 -45 -55>, angularTarget=<0 0 0>}
      localJointOffset=<4 0 0>
     }]
    planarLimits [{drivingBone {boneName=pack}, planeTransform {location=<0.4012 -0.5662 0>, rotation=<rot 0 10 -90.0003>}}]
    sphericalLimits [
     {limitRadius=14, drivingBone {boneName=spine_01}, sphereLocalOffset=<0 0 0>}
     {limitRadius=14, drivingBone {boneName=spine_04}, sphereLocalOffset=<7.6385 0 0>}
     {limitRadius=14, drivingBone {boneName=upperarm_l}, sphereLocalOffset=<0 0 0>}
    ]
   }
   n3 = Anim Dynamics - Bone: dyn_strap_02_L!(ComponentPose=@n2, Alpha=1, GravityOverride:Vector=@n1, ExternalForce="0,0,0") -> (Pose) {
    angularBiasOverride=1
    angularDampingOverride=1
    angularSpringConstant=1000
    bAngularSpring=true
    bOverrideAngularBias=true
    bOverrideAngularDamping=true
    bUseGravityOverride=true
    bUsePlanarLimit=false
    bUseSphericalLimits=true
    simulationSpace=RootRelative
    alphaBoolBlend {blendOption=Linear}
    boundBone {boneName=dyn_strap_02_L}
    componentAppliedLinearAccClamp=<3000 3000 3000>
    componentLinearAccScale=<0.4 0.5 0.5>
    componentLinearVelScale=<0.5 0.5 0.5>
    physicsBodyDefinitions [{
      collisionType=CustomSphere
      sphereCollisionRadius=3
      boundBone {boneName=dyn_strap_02_L}
      boxExtents=<15 9 9>
      constraintSetup {angularLimitsMax=<0 45 45>, angularLimitsMin=<0 -45 -45>, angularTarget=<0 0 0>}
      localJointOffset=<6 0 0>
     }]
    planarLimits [{drivingBone {boneName=pack}, planeTransform {location=<0.3174 -0.0911 0>, rotation=<rot 0.0001 10 -90.0003>}}]
    sphericalLimits [
     {limitRadius=12, drivingBone {boneName=spine_01}, sphereLocalOffset=<0 0 0>}
     {limitRadius=12, drivingBone {boneName=spine_04}, sphereLocalOffset=<0 0 0>}
     {limitRadius=12, drivingBone {boneName=upperarm_l}, sphereLocalOffset=<0 0 0>}
    ]
   }
   n4 = Anim Dynamics - Bone: dyn_strap_03_L!(ComponentPose=@n3, Alpha=1, GravityOverride:Vector=@n1, ExternalForce="0,0,0") -> (Pose) {
    angularBiasOverride=1
    angularDampingOverride=1
    angularSpringConstant=1000
    bAngularSpring=true
    bOverrideAngularBias=true
    bOverrideAngularDamping=true
    bUseGravityOverride=true
    bUsePlanarLimit=false
    bUseSphericalLimits=true
    simulationSpace=RootRelative
    alphaBoolBlend {blendOption=Linear}
    boundBone {boneName=dyn_strap_03_L}
    componentAppliedLinearAccClamp=<3000 3000 3000>
    componentLinearAccScale=<0.4 0.5 0.5>
    componentLinearVelScale=<0.5 0.5 0.5>
    physicsBodyDefinitions [{
      collisionType=CustomSphere
      sphereCollisionRadius=3
      boundBone {boneName=dyn_strap_03_L}
      boxExtents=<14 8 8>
      constraintSetup {angularLimitsMax=<0 45 45>, angularLimitsMin=<0 -45 -45>, angularTarget=<0 0 0>}
      localJointOffset=<7 0 0>
     }]
    planarLimits [{drivingBone {boneName=pack}, planeTransform {location=<0.3174 -0.0911 0>, rotation=<rot 0.0001 10 -90.0003>}}]
    sphericalLimits [
     {limitRadius=12.5, drivingBone {boneName=spine_01}, sphereLocalOffset=<0 0 0>}
     {limitRadius=12.5, drivingBone {boneName=spine_04}, sphereLocalOffset=<0 0 0>}
     {limitRadius=14, drivingBone {boneName=upperarm_l}, sphereLocalOffset=<0 0 0>}
    ]
   }
   n5 = Anim Dynamics - Bone: dyn_strap_01_R!(ComponentPose=@n4, Alpha=1, GravityOverride:Vector=@n1, ExternalForce="0,0,0") -> (Pose) {
    angularBiasOverride=1
    angularDampingOverride=1
    angularSpringConstant=3000
    bAngularSpring=true
    bOverrideAngularBias=true
    bOverrideAngularDamping=true
    bUseGravityOverride=true
    simulationSpace=RootRelative
    alphaBoolBlend {blendOption=Linear}
    boundBone {boneName=dyn_strap_01_R}
    componentAppliedLinearAccClamp=<3000 3000 3000>
    componentLinearAccScale=<0.4 0.5 0.5>
    componentLinearVelScale=<0.5 0.5 0.5>
    physicsBodyDefinitions [{
      collisionType=CustomSphere
      sphereCollisionRadius=2.2
      boundBone {boneName=dyn_strap_01_R}
      boxExtents=<18 14 14>
      constraintSetup {angularLimitsMax=<0 45 10>, angularLimitsMin=<0 -45 -55>, angularTarget=<0 0 0>}
      localJointOffset=<5 0 0>
     }]
    planarLimits [{drivingBone {boneName=pack}, planeTransform {location=<0.354 -0.2986 0>, rotation=<rot 0.0001 10 -90.0003>}}]
    sphericalLimits [
     {limitRadius=14, drivingBone {boneName=spine_01}, sphereLocalOffset=<0 0 0>}
     {limitRadius=14, drivingBone {boneName=spine_04}, sphereLocalOffset=<7.6385 0 0>}
     {limitRadius=13, drivingBone {boneName=pack}, sphereLocalOffset=<-18.4199 10.4376 -24.6877>}
     {limitRadius=15, drivingBone {boneName=pack}, sphereLocalOffset=<0 13.796 -16.2864>}
    ]
   }
   n6 = Anim Dynamics - Bone: dyn_strap_02_R!(ComponentPose=@n5, Alpha=1, GravityOverride:Vector=@n1, ExternalForce="0,0,0") -> (Pose) {
    angularBiasOverride=1
    angularDampingOverride=1
    angularSpringConstant=3000
    bAngularSpring=true
    bOverrideAngularBias=true
    bOverrideAngularDamping=true
    bUseGravityOverride=true
    simulationSpace=RootRelative
    alphaBoolBlend {blendOption=Linear}
    boundBone {boneName=dyn_strap_02_R}
    componentAppliedLinearAccClamp=<3000 3000 3000>
    componentLinearAccScale=<0.4 0.5 0.5>
    componentLinearVelScale=<0.5 0.5 0.5>
    physicsBodyDefinitions [{
      collisionType=CustomSphere
      sphereCollisionRadius=3
      boundBone {boneName=dyn_strap_02_R}
      boxExtents=<18 12 12>
      constraintSetup {angularLimitsMax=<0 45 45>, angularLimitsMin=<0 -45 -45>, angularTarget=<0 0 0>}
      localJointOffset=<6 0 0>
     }]
    planarLimits [{drivingBone {boneName=pack}, planeTransform {location=<0.152 0.847 0>, rotation=<rot 0.0001 10 -90.0003>}}]
    sphericalLimits [
     {limitRadius=14, drivingBone {boneName=spine_01}, sphereLocalOffset=<0 1.1731 -0.0132>}
     {limitRadius=14, drivingBone {boneName=spine_04}, sphereLocalOffset=<7.6385 0 0>}
     {limitRadius=12.5, drivingBone {boneName=pack}, sphereLocalOffset=<-18.4199 9.1785 -24.6877>}
    ]
   }
   n7 = Anim Dynamics - Bone: dyn_strap_03_R!(ComponentPose=@n6, Alpha=1, GravityOverride:Vector=@n1, ExternalForce="0,0,0") -> (Pose) {
    angularBiasOverride=1
    angularDampingOverride=1
    angularSpringConstant=3000
    bAngularSpring=true
    bOverrideAngularBias=true
    bOverrideAngularDamping=true
    bUseGravityOverride=true
    bUsePlanarLimit=false
    bUseSphericalLimits=true
    simulationSpace=RootRelative
    alphaBoolBlend {blendOption=Linear}
    boundBone {boneName=dyn_strap_03_R}
    componentAppliedLinearAccClamp=<3000 3000 3000>
    componentLinearAccScale=<0.4 0.5 0.5>
    componentLinearVelScale=<0.5 0.5 0.5>
    physicsBodyDefinitions [{
      collisionType=CustomSphere
      sphereCollisionRadius=3
      boundBone {boneName=dyn_strap_03_R}
      boxExtents=<17 11 11>
      constraintSetup {angularLimitsMax=<0 45 45>, angularLimitsMin=<0 -45 -45>, angularTarget=<0 0 0>}
      localJointOffset=<7 0 0>
     }]
    planarLimits [{drivingBone {boneName=pack}, planeTransform {location=<0.3174 -0.0911 0>, rotation=<rot 0.0001 10 -90.0003>}}]
    sphericalLimits [
     {limitRadius=14, drivingBone {boneName=spine_01}, sphereLocalOffset=<0 1.4851 -0.5862>}
     {limitRadius=14, drivingBone {boneName=spine_04}, sphereLocalOffset=<7.6385 0 0>}
     {limitRadius=14, drivingBone {boneName=pack}, sphereLocalOffset=<-18.4199 11.9087 -24.6877>}
    ]
   }
   n8 = RigidBody!(
    ComponentPose=Local To Component(LocalPose=Use cached pose 'before simulation'())
    Alpha
    OverrideWorldGravity:Vector=Get Gravity Override()
    ExternalForce
   ) -> (Pose) {
    bClampLinearTranslationLimitToRefPose=true
    bOverrideWorldGravity=true
    overridePhysicsAsset=&/M_MED_ScorpionZero_Pack_PhysicsAsset
    simulationSpace=BaseBoneSpace
    alphaBoolBlend {blendOption=Linear}
    baseBoneRef {boneName=pack}
    componentAppliedLinearAccClamp=<3000 3000 3000>
    componentLinearAccScale=<0.4 0.4 0.5>
    componentLinearVelScale=<0.4 0.4 0.5>
    overrideWorldGravity {z=-980}
   }
   Output Pose(Result=Blend Poses by bool(BlendPose_0=Component To Local(ComponentPose=@n8), BlendPose_1=Component To Local(ComponentPose=@n7), bActiveValue=Get bRigidBodyEnabled()) {blendType=Linear}) {layerGroup=None}
   n9 = "Transform (Modify) Bone - Bone: pack"(ComponentPose=Local To Component(LocalPose=Input Pose()), Translation="-1.3,0,-1.7", Rotation="0,0,0") {
    translationMode=BMM_Additive
    translationSpace=BCS_BoneSpace
    alphaBoolBlend {blendOption=Linear}
    boneToModify {boneName=pack}
    translation {x=-1.3, z=-1.7}
   }
   n10 = "Transform (Modify) Bone - Bone: spine_01"(
    ComponentPose=@n9
    Alpha=1
    Translation=Get spine_01 translation() -> (spine_01 translation:Vector)
    Rotation=Get spine_01rotation() -> (spine_01rotation:Rotator)
    Scale="1,1,1"
   ) {
    rotationMode=BMM_Replace
    rotationSpace=BCS_WorldSpace
    scaleMode=BMM_Additive
    translationMode=BMM_Replace
    translationSpace=BCS_WorldSpace
    alphaBoolBlend {blendInTime=1, blendOption=Linear, blendOutTime=1}
    boneToModify {boneName=spine_01}
    rotation {roll=20}
   }
   n11 = "Transform (Modify) Bone - Bone: spine_04"(
    ComponentPose=@n10
    Alpha=1
    Translation=Get spine 04 translation() -> (spine 04 translation:Vector)
    Rotation=Get spine 04 rotation() -> (spine 04 rotation:Rotator)
    Scale="1,1,1"
   ) {
    rotationMode=BMM_Replace
    rotationSpace=BCS_WorldSpace
    scaleMode=BMM_Additive
    translationMode=BMM_Replace
    translationSpace=BCS_WorldSpace
    alphaBoolBlend {blendInTime=1, blendOption=Linear, blendOutTime=1}
    boneToModify {boneName=spine_04}
    rotation {roll=20}
   }
   n12 = "Transform (Modify) Bone - Bone: upperarm_l"(
    ComponentPose=@n11
    Alpha=1
    Translation=Get upperarm_l translation() -> (upperarm_l translation:Vector)
    Rotation=Get upperarm_l rotation() -> (upperarm_l rotation:Rotator)
    Scale="1,1,1"
   ) {
    rotationMode=BMM_Replace
    rotationSpace=BCS_WorldSpace
    scaleMode=BMM_Additive
    translationMode=BMM_Replace
    translationSpace=BCS_WorldSpace
    alphaBoolBlend {blendInTime=1, blendOption=Linear, blendOutTime=1}
    boneToModify {boneName=upperarm_l}
    rotation {roll=20}
   }
   Save cached pose 'before simulation'(Pose=Component To Local(ComponentPose=@n12)) {cacheName=before simulation}
  ]}
 EventGraph {nodes [
   Event BlueprintUpdateAnimation()
   > Is Valid(InputObject=Get BodyMesh() -> (BodyMesh:Skeletal Mesh Component Object Reference))
   n1 = BreakTransform(InTransform=GetSocketTransform(self=Get BodyMesh(), InSocketName=spine_01, TransformSpace=RTS_World))
   > Set spine_01 translation(spine_01 translation:Vector=@n1) -> (Output_Get:Vector)
   > Set spine_01rotation(spine_01rotation:Rotator=@n1.Rotation) -> (Output_Get:Rotator)
   n2 = BreakTransform(InTransform=GetSocketTransform(self=Get BodyMesh(), InSocketName=spine_04, TransformSpace=RTS_World))
   > Set spine 04 translation(spine 04 translation:Vector=@n2) -> (Output_Get:Vector)
   > Set spine 04 rotation(spine 04 rotation:Rotator=@n2.Rotation) -> (Output_Get:Rotator)
   n3 = BreakTransform(InTransform=GetSocketTransform(self=Get BodyMesh(), InSocketName=upperarm_l, TransformSpace=RTS_World))
   > Set upperarm_l translation(upperarm_l translation:Vector=@n3) -> (Output_Get:Vector)
   > Set upperarm_l rotation(upperarm_l rotation:Rotator=@n3.Rotation) -> (Output_Get:Rotator)
   > n4 = Gravity Override State Function DEPRECATED(CharacterPartAnimInstance=Self-Reference(), GameTimeInSeconds=GetGameTimeInSeconds(), GravityOverrideParams=Get Gravity Override Params())
   > Set Gravity Override(Gravity Override=@n4)
   n5 = Event OnCharacterCustomizationLoaded()
   > n6 = Set BodyMesh(BodyMesh=Get Array Item with EFort Custom Part Type(Array=@n5.CustomMeshParts))
   n7 = Get Dynamic Material Instance(execute=[@n11.then, @n6.then], SkeletalMeshComponent=GetPartSkeletalMeshComponent())
   n8 = PickaxeItemDefinitionHasAnyMetaTag(Tags=Make Array("[0]"=(TagName="Cosmetics.MetaTags.ScorpionZero")), bRequiresEquipped=true)
   > n9 = Branch(Condition=@n8)
   > n10 = Branch(Condition=Get bIsSkyDiving())
   SetScalarParameterValue(execute=[@n12.then, @n13.else, @n10.then, @n9.else], self=@n7, ParameterName=Hide Element 01, Value=0)
   SetScalarParameterValue(execute=[@n13.then, @n10.else], self=@n7, ParameterName=Hide Element 01, Value=1)
   n11 = Event OnWeaponChanged()
   n12 = Event OnBeginSkydiving()
   Event OnEndSkydiving()
   > n13 = Branch(Condition=@n8)
  ]}
]
variables [
 As Player Pawn Athena {type=PlayerPawn_Athena_C}
 BodyMesh {type=SkeletalMeshComponent}
 Gravity Override {type=Vector}
 Gravity Override Params {type=GravityOverrideParamsStruct}
 spine 04 rotation {type=Rotator}
 spine 04 translation {type=Vector}
 spine_01 translation {type=Vector}
 spine_01rotation {type=Rotator}
 upperarm_l rotation {type=Rotator}
 upperarm_l translation {type=Vector}
]
